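# Player animation layer for a Figment ghost form, replacing the final pose with GhostRock locomotion and butterfly look-at.
# A cardinal-direction idle/start/jog/stop/pivot state machine feeds aim offset and lean additives into a cached pose.
# Montage slots, a curve-masked upper-body blend, butterfly aim offset and curve-driven weapon-bone copy follow.
# Update logic clamps speed and play rates, and derives look-at angles from ButterFlyLocation, toggling look-at by montage playback.
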AnimBlueprint /Figment_S06/Athena/Prototype/Blueprints/Island/Figment_Butterfly_Ghost_Layer_AnimBP
target_skeleton=&Fortnite_M_Avg_Player_Skeleton
graphs [
 AnimGraph {nodes [Output Pose() {layerGroup=None}]}
 CheckAnimIsValid {nodes [
   n1 = Inputs!() -> (In:Exec, In Pose:Anim Sequence Object Reference, FallBackPose:Anim Sequence Object Reference)
   > n2 = Cast To AnimSequence!(
    execute
    Object=SelectObject!(A:Object Reference=@n1, B:Object Reference=@n1.FallBackPose, bSelectA:bool=IsValid(Object=@n1)) -> (ReturnValue:Object Reference)
   ) -> (then, CastFailed, AsAnim Sequence:Anim Sequence Object Reference)
   Outputs!(then:Exec=[@n2.CastFailed, @n2.then], OutPose:Anim Sequence Object Reference=@n2)
  ]}
 CheckBlendspaceIsValid {nodes [
   n1 = Inputs!() -> (in:Exec, In Blendspace:Blend Space Object Reference, FallBack Blendspace:Blend Space Object Reference)
   > n2 = Cast To BlendSpace!(execute, Object=SelectObject(A=@n1, B=@n1.FallBack Blendspace, bSelectA=IsValid(Object=@n1))) -> (then, CastFailed, AsBlend Space:Blend Space Object Reference) {is=title K2Node_DynamicCast}
   Outputs!(then=[@n2.CastFailed, @n2.then], Out Blendspace:Blend Space Object Reference=@n2)
  ]}
 EventGraph {nodes [
   Event BlueprintUpdateAnimation()
   > n1 = Sequence!(execute) -> (then_0, then_1, then_2:Exec)
   n2 = Get ShouldLookAtButterfly() -> (ShouldLookAtButterfly:bool)
   n3 = Get bIsMoving2D()
   > Set EnableLeanAdditives(EnableLeanAdditives:bool=NOT Boolean(A=AND Boolean(A=@n2, B=@n3))) -> (Output_Get:bool)
   n4 = TryGetPawnOwner()
   > Is Valid(InputObject=@n4)
   > Branch(Condition="Not Equal (Vector)"!(A:Vector=Get ButterFlyLocation() -> (ButterFlyLocation:Vector), B:Vector="0, 0, 0", ErrorTolerance:float=0.01) -> (ReturnValue:bool))
   > Set ShouldLookAtButterfly(ShouldLookAtButterfly:bool=NOT Boolean(A=IsAnyMontagePlaying!(self:Anim Instance Object Reference) -> (ReturnValue:bool))) -> (Output_Get:bool)
   n5 = FindLookAtRotation!(Start:"Vector (by ref)"=Get Actor Location(self=@n4), Target:"Vector (by ref)"=Get ButterFlyLocation()) -> (ReturnValue:Rotator)
   > Set ButterflyLookAtYaw(ButterflyLookAtYaw:double="Delta (Rotator)"!(A:Rotator=GetBaseAimRotation!(self:Pawn Object Reference=@n4) -> (ReturnValue:Rotator), B:Rotator=@n5) -> (ReturnValue_Roll:float, ReturnValue_Pitch:float, ReturnValue_Yaw:float).ReturnValue_Yaw) -> (Output_Get:double)
   > Set ButterflyLookAtPitch(ButterflyLookAtPitch:double=BreakRotator(InRot=@n5).Pitch) -> (Output_Get:double)
   Set ClampedSpeed2D(execute=@n1.then_1, ClampedSpeed2D:double="Clamp (Float)"(Value=Get Speed2D() -> (Speed2D:float), Min=0, Max=500)) -> (Output_Get:double)
   n6 = Get StartSpeedWarping() -> (StartSpeedWarping:Fort Anim Input Speed Warping Structure)
   > Set ClampedStartPlayRate(ClampedStartPlayRate:double="Clamp (Float)"(
     Value=Break Fort Anim Input Speed Warping!(FortAnimInput_SpeedWarping:"Fort Anim Input Speed Warping Structure (by ref)"=@n6) -> (SpeedWarpingAmount:float=~, PlayRate:float=~) {is=title K2Node_BreakStruct}.PlayRate
     Min=0
     Max=1
    )) -> (Output_Get:double)
   n7 = Get SpeedWarping() -> (SpeedWarping:Fort Anim Input Speed Warping Structure)
   > Set ClampedPlayRate(ClampedPlayRate:double="Clamp (Float)"(Value=Break Fort Anim Input Speed Warping(FortAnimInput_SpeedWarping=@n7) -> (SpeedWarpingAmount=~, PlayRate=~).PlayRate, Min=0, Max=1)) -> (Output_Get:double)
   > n8 = Branch(Condition="float > float"(A=GetCurveValue(CurveName=UpperBodyNoAdditive)))
   > Set UpperBodyNoAdditivesMaskWeight(UpperBodyNoAdditivesMaskWeight:double=1) -> (Output_Get:double)
   Set UpperBodyNoAdditivesMaskWeight(execute=@n8.else, UpperBodyNoAdditivesMaskWeight=0)
   AnimNotify_playFacialAnim() {is=title K2Node_Event}
   > n9 = Branch(Condition=Cast To FortGameStateZone!(Object=GetGameState!() -> (ReturnValue:Game State Base Object Reference)) -> (AsFort Game State Zone:Fort Game State Zone Object Reference, bSuccess:bool=~) {is=title K2Node_DynamicCast}.bSuccess)
   Call Menu Screen Dispatcher!(execute:Exec=@n9.else, self:Figment Butterfly Ghost Layer Anim BP Object Reference, CharPart:Fort Player Anim Instance Object Reference) -> (then:Exec) {is=title K2Node_CallDelegate}
   CharPartEvent!() -> (OutputDelegate, then, CharPart:Fort Player Anim Instance Object Reference) {is=title K2Node_CustomEvent}
  ]}
 MenuScreenDispatcher {nodes [MenuScreenDispatcher!() -> (then, CharPart:Fort Player Anim Instance Object Reference) {is=title K2Node_FunctionEntry}]}
 MovementMode_AimOffset {nodes [
   Input Pose() {name=InAimOffsetsPose}
   Input Pose() {name=InSourcePosePreAimOffsets}
   Output Pose() {layerGroup=MovementMode}
  ]}
 MovementMode_BaseLocomotion {nodes [
   Input Pose() {name=InPoseBaseLocomotion}
   Output Pose() {layerGroup=MovementMode}
  ]}
 MovementMode_FinalPoseOverride {nodes [
   Input Pose() {name=InFinalPose}
   n1 = Layered blend per bone(BasePose=Use cached pose 'LocomotionPose'(), BlendPoses_0=Slot 'UpperBody'(Source=Use cached pose 'LocomotionPose'()) {slotName=UpperBody}) {
    bMeshSpaceRotationBlend=true
    layerSetup [{branchFilters [{blendDepth=5, boneName=spine_01}, {boneName=clavicle_l}, {boneName=clavicle_r}, {boneName=ik_hand_root}]}]
   }
   n2 = Layered blend per bone(
    BasePose=Slot 'FullBodyNoAdditive'(Source=@n1) {slotName=FullBodyNoAdditive}
    BlendPoses_0=Use cached pose 'LocomotionPose'()
    BlendWeights_0=Get UpperBodyNoAdditivesMaskWeight() -> (UpperBodyNoAdditivesMaskWeight:double)
   ) {
    bMeshSpaceRotationBlend=true
    curveBlendOption=BlendByWeight
    layerSetup [{branchFilters [{boneName=Root}, {blendDepth=-1, boneName=Ik_hand_root}, {blendDepth=-4, boneName=Spine_01}]}]
   }
   n3 = AimOffset Player 'ButterflyLookAt_AO'!(
    X:float=Get ButterflyLookAtYaw() -> (ButterflyLookAtYaw:double)
    Y:float=Get ButterflyLookAtPitch() -> (ButterflyLookAtPitch:double)
    BasePose:Pose Link Structure=Slot 'FullBody'(Source=@n2) {slotName=FullBody}
    bAlphaBoolEnabled:bool=Get ShouldLookAtButterfly()
   ) -> (Pose:Pose Link Structure) {
    alphaInputType=Bool
    blendSpace=/Game/Animation/Game/MainPlayer/Poses/Medium/Male/AO/ButterflyLookAt_AO
    alphaBoolBlend {blendInTime=0.5, blendOption=Cubic, blendOutTime=3}
   }
   n4 = Copy Bone - Source Bone: weapon_r - Target Bone: attach!(
    ComponentPose:Component Space Pose Link Structure=Local To Component(LocalPose=@n3)
    AlphaCurveName:Name=DragonSlash
    bCopyTranslation:bool=true
    bCopyRotation:bool=true
    bCopyScale:bool=False
   ) -> (Pose:Component Space Pose Link Structure) {
    is=title AnimGraphNode_CopyBone
    alphaInputType=Curve
    alphaBoolBlend {blendOption=Linear}
    sourceBone {boneName=weapon_r}
    targetBone {boneName=attach}
   }
   n5 = Component To Local(ComponentPose=@n4)
   Output Pose(Result=Slot 'FullBody'(Source=Modify Curve(SourcePose=@n5, CurveValues_0=1, Alpha=1) {curveNames [BothLegFK], curveValues [0]}) {slotName=FullBody}) {layerGroup=MovementMode}
   n6 = Get YawDeltaSmoothed() -> (YawDeltaSmoothed:float)
   AnimGraphNode_StateMachine_0 = Locomotion() {is=title AnimGraphNode_StateMachine}
   AnimGraphNode_StateMachine_1 = LocomotionLowerBody() {is=title AnimGraphNode_StateMachine}
   AnimGraphNode_StateMachine_2 = FullBodyAdditives() {is=title AnimGraphNode_StateMachine}
   n7 = Blend Poses by int!(
    BlendPose_0:Pose Link Structure=@AnimGraphNode_StateMachine_0
    BlendPose_1:Pose Link Structure=@AnimGraphNode_StateMachine_1
    BlendPose_2:Pose Link Structure=@AnimGraphNode_StateMachine_2
    BlendTime_0:float=0.1
    BlendTime_1:float=0.1
    BlendTime_2:float=0.1
    ActiveChildIndex:int=0
   ) -> (Pose:Pose Link Structure) {
    is=title AnimGraphNode_BlendListByInt
    blendType=Linear
    blendPose [{linkId=-1, sourceLinkId=-1}, {linkId=-1, sourceLinkId=-1}, {linkId=-1, sourceLinkId=-1}]
    blendTime [0.1,0.1,0.1]
   }
   n8 = AimOffset Player 'GhostRock_AO'!(
    X:float=Get AimYaw() -> (AimYaw:float)
    Y:float=Get AimPitch()
    BasePose:Pose Link Structure=@n7
    bAlphaBoolEnabled:bool=NOT Boolean(A=Get ShouldLookAtButterfly())
   ) -> (Pose:Pose Link Structure) {
    alphaInputType=Bool
    blendSpace=/Game/Animation/Game/MainPlayer/Poses/Medium/Male/AO/GhostRock_AO
    alphaBoolBlend {blendInTime=0.2, blendOption=Linear, blendOutTime=0.2}
    alphaScaleBiasClamp {bias=1, scale=-1}
   }
   Save cached pose 'LocomotionPose'(Pose=Apply Additive!(
     Base=@n8
     Additive=Blendspace Player 'GhostRock_Idle_Leans_BS'!(X=@n6) -> (Pose) {blendSpace=/Game/Animation/Game/MainPlayer/Combat/Gadgets/Medium/Male/GhostRock/GhostRock_Idle_Leans_BS}
     bAlphaBoolEnabled:bool=Get EnableLeanAdditives() -> (EnableLeanAdditives:bool)
    ) -> (Pose) {alphaInputType=Bool, alphaBoolBlend {blendInTime=0.2, blendOption=Linear, blendOutTime=0.2}}) {cacheName=LocomotionPose}
  ]}
 Locomotion {
  entry=IdleAdditive
  path=MovementMode_FinalPoseOverride.AnimGraphNode_StateMachine_0.Locomotion
  ids=id
  nodes [
   AnimStateNode_0 = IdleAdditive() {is=title AnimStateNode}
   AnimStateNode_1 = JogStartAdditive() {is=title AnimStateNode}
   AnimStateNode_2 = JogStopAdditive() {is=title AnimStateNode}
   AnimStateNode_3 = JogPostPivotAdditive() {is=title AnimStateNode, bAlwaysResetOnEntry=true}
   AnimStateNode_6 = JogPrePivotAdditive() {is=title AnimStateNode, bAlwaysResetOnEntry=true}
   AnimStateNode_7 = MovementAdditive() {is=title AnimStateNode}
   AnimStateTransitionNode_0 = AnimStateTransitionNode()
   AnimStateTransitionNode_10 = AnimStateTransitionNode() {
    crossfadeDuration=0.4
    priorityOrder=2
    blendProfileWrapper {blendProfile=&Fortnite_M_Avg_Player_Skeleton:FastFeet}
   }
   AnimStateTransitionNode_11 = AnimStateTransitionNode() {crossfadeDuration=0.25, priorityOrder=2}
   AnimStateTransitionNode_12 = AnimStateTransitionNode() {crossfadeDuration=0.1667}
   AnimStateTransitionNode_13 = AnimStateTransitionNode() {bAutomaticRuleBasedOnSequencePlayerInState=true, crossfadeDuration=0.5, priorityOrder=2}
   AnimStateTransitionNode_2 = AnimStateTransitionNode() {bAutomaticRuleBasedOnSequencePlayerInState=true, crossfadeDuration=0.5}
   AnimStateTransitionNode_25 = AnimStateTransitionNode() {crossfadeDuration=0.0667}
   AnimStateTransitionNode_26 = AnimStateTransitionNode()
   AnimStateTransitionNode_27 = AnimStateTransitionNode() {bAutomaticRuleBasedOnSequencePlayerInState=true, crossfadeDuration=0.4}
   AnimStateTransitionNode_28 = AnimStateTransitionNode() {
    crossfadeDuration=0.15
    blendProfileWrapper {blendProfile=&Fortnite_M_Avg_Player_Skeleton:FastFeet}
   }
   AnimStateTransitionNode_3 = AnimStateTransitionNode()
   AnimStateTransitionNode_4 = AnimStateTransitionNode() {crossfadeDuration=0.25}
   AnimStateTransitionNode_5 = AnimStateTransitionNode()
   AnimStateTransitionNode_6 = AnimStateTransitionNode() {blendProfileWrapper {blendProfile=&Fortnite_M_Avg_Player_Skeleton:FastFeet}}
   AnimStateTransitionNode_7 = AnimStateTransitionNode() {crossfadeDuration=0.4, priorityOrder=2}
   AnimStateTransitionNode_8 = AnimStateTransitionNode() {
    bAutomaticRuleBasedOnSequencePlayerInState=true
    crossfadeDuration=0.75
    blendProfileWrapper {blendProfile=&Fortnite_M_Avg_Player_Skeleton:FastFeet}
   }
   AnimStateTransitionNode_9 = AnimStateTransitionNode() {priorityOrder=3, blendProfileWrapper {blendProfile=&Fortnite_M_Avg_Player_Skeleton:FastFeet}}
  ]
  links [
   {
    dst=JogPrePivotAdditive
    rule=MovementMode_FinalPoseOverride.AnimGraphNode_StateMachine_0.Locomotion.AnimStateTransitionNode_4.Transition
    src=IdleAdditive
    via=AnimStateTransitionNode_4
   }
   {
    dst=JogStartAdditive
    rule=MovementMode_FinalPoseOverride.AnimGraphNode_StateMachine_0.Locomotion.AnimStateTransitionNode_0.Transition
    src=IdleAdditive
    via=AnimStateTransitionNode_0
   }
   {
    dst=JogStopAdditive
    rule=MovementMode_FinalPoseOverride.AnimGraphNode_StateMachine_0.Locomotion.AnimStateTransitionNode_12.Transition
    src=IdleAdditive
    via=AnimStateTransitionNode_12
   }
   {
    dst=MovementAdditive
    rule=MovementMode_FinalPoseOverride.AnimGraphNode_StateMachine_0.Locomotion.AnimStateTransitionNode_10.AnimationTransitionGraph_114
    src=IdleAdditive
    via=AnimStateTransitionNode_10
   }
   {
    dst=JogPrePivotAdditive
    rule=MovementMode_FinalPoseOverride.AnimGraphNode_StateMachine_0.Locomotion.AnimStateTransitionNode_28.AnimationTransitionGraph_0
    src=JogPostPivotAdditive
    via=AnimStateTransitionNode_28
   }
   {
    dst=MovementAdditive
    rule=MovementMode_FinalPoseOverride.AnimGraphNode_StateMachine_0.Locomotion.AnimStateTransitionNode_26.Transition
    src=JogPostPivotAdditive
    via=AnimStateTransitionNode_26
   }
   {
    dst=MovementAdditive
    rule=MovementMode_FinalPoseOverride.AnimGraphNode_StateMachine_0.Locomotion.AnimStateTransitionNode_27.Transition
    src=JogPostPivotAdditive
    via=AnimStateTransitionNode_27
   }
   {
    dst=JogPostPivotAdditive
    rule=MovementMode_FinalPoseOverride.AnimGraphNode_StateMachine_0.Locomotion.AnimStateTransitionNode_25.AnimationTransitionGraph_0
    src=JogPrePivotAdditive
    via=AnimStateTransitionNode_25
   }
   {
    dst=MovementAdditive
    rule=MovementMode_FinalPoseOverride.AnimGraphNode_StateMachine_0.Locomotion.AnimStateTransitionNode_13.Transition
    src=JogPrePivotAdditive
    via=AnimStateTransitionNode_13
   }
   {
    dst=MovementAdditive
    rule=MovementMode_FinalPoseOverride.AnimGraphNode_StateMachine_0.Locomotion.AnimStateTransitionNode_5.Transition
    src=JogPrePivotAdditive
    via=AnimStateTransitionNode_5
   }
   {
    dst=MovementAdditive
    rule=MovementMode_FinalPoseOverride.AnimGraphNode_StateMachine_0.Locomotion.AnimStateTransitionNode_2.Transition
    src=JogStartAdditive
    via=AnimStateTransitionNode_2
   }
   {
    dst=MovementAdditive
    rule=MovementMode_FinalPoseOverride.AnimGraphNode_StateMachine_0.Locomotion.AnimStateTransitionNode_3.Transition
    src=JogStartAdditive
    via=AnimStateTransitionNode_3
   }
   {
    dst=IdleAdditive
    rule=MovementMode_FinalPoseOverride.AnimGraphNode_StateMachine_0.Locomotion.AnimStateTransitionNode_11.Transition
    src=JogStopAdditive
    via=AnimStateTransitionNode_11
   }
   {
    dst=IdleAdditive
    rule=MovementMode_FinalPoseOverride.AnimGraphNode_StateMachine_0.Locomotion.AnimStateTransitionNode_8.Transition
    src=JogStopAdditive
    via=AnimStateTransitionNode_8
   }
   {
    dst=IdleAdditive
    rule=MovementMode_FinalPoseOverride.AnimGraphNode_StateMachine_0.Locomotion.AnimStateTransitionNode_9.AnimationTransitionGraph_113
    src=MovementAdditive
    via=AnimStateTransitionNode_9
   }
   {
    dst=JogPrePivotAdditive
    rule=MovementMode_FinalPoseOverride.AnimGraphNode_StateMachine_0.Locomotion.AnimStateTransitionNode_6.Transition
    src=MovementAdditive
    via=AnimStateTransitionNode_6
   }
   {
    dst=JogStopAdditive
    rule=MovementMode_FinalPoseOverride.AnimGraphNode_StateMachine_0.Locomotion.AnimStateTransitionNode_7.Transition
    src=MovementAdditive
    via=AnimStateTransitionNode_7
   }
  ]
 }
 IdleAdditive {
  path=MovementMode_FinalPoseOverride.AnimGraphNode_StateMachine_0.Locomotion.AnimStateNode_0.IdleAdditive
  nodes [Output Animation Pose(Result=Sequence Player 'GhostRock_Idle_CMM_NoAdditive'() {sequence=/Game/Animation/Game/MainPlayer/Combat/Gadgets/Medium/Male/GhostRock/NoAdditives/GhostRock_Idle_CMM_NoAdditive})]
 }
 JogStartAdditive {
  path=MovementMode_FinalPoseOverride.AnimGraphNode_StateMachine_0.Locomotion.AnimStateNode_1.JogStartAdditive
  nodes [
   n1 = Get ClampedStartPlayRate() -> (ClampedStartPlayRate:double)
   n2 = Sequence Player 'GhostRock_Jog_S_Start_CMM_NoAdditive'!(PlayRate:float=@n1) -> (Pose) {
    bLoopAnimation=false
    groupName=Locomotion
    groupRole=TransitionLeader
    method=SyncGroup
    sequence=/Game/Animation/Game/MainPlayer/Locomotion/Medium/Male/Jog/GhostRock/LocomotionNoAdditive/GhostRock_Jog_S_Start_CMM_NoAdditive
   }
   n3 = Sequence Player 'GhostRock_Jog_N_Start_CMM_NoAdditive'!(PlayRate:float=@n1) -> (Pose) {
    bLoopAnimation=false
    groupName=Locomotion
    groupRole=TransitionLeader
    method=SyncGroup
    sequence=/Game/Animation/Game/MainPlayer/Locomotion/Medium/Male/Jog/GhostRock/LocomotionNoAdditive/GhostRock_Jog_N_Start_CMM_NoAdditive
   }
   n4 = Sequence Player 'GhostRock_Jog_W_Start_CMM_NoAdditive'!(PlayRate:float=@n1) -> (Pose) {
    bLoopAnimation=false
    groupName=Locomotion
    groupRole=TransitionLeader
    method=SyncGroup
    sequence=/Game/Animation/Game/MainPlayer/Locomotion/Medium/Male/Jog/GhostRock/LocomotionNoAdditive/GhostRock_Jog_W_Start_CMM_NoAdditive
   }
   n5 = Sequence Player 'GhostRock_Jog_E_Start_CMM_NoAdditive'!(PlayRate:float=@n1) -> (Pose) {
    bLoopAnimation=false
    groupName=Locomotion
    groupRole=TransitionLeader
    method=SyncGroup
    sequence=/Game/Animation/Game/MainPlayer/Locomotion/Medium/Male/Jog/GhostRock/LocomotionNoAdditive/GhostRock_Jog_E_Start_CMM_NoAdditive
   }
   n6 = Get StartAnimBlendTime() -> (StartAnimBlendTime:double)
   n7 = "Blend Poses (EFortCardinalDirection)"!(
    BlendPose_0:Pose Link Structure=Rotate Root Bone!(BasePose:Pose Link Structure=@n3, Yaw:float=Get StartAnimDeltaAngleNorth() -> (StartAnimDeltaAngleNorth:float)) -> (Pose:Pose Link Structure) {is=title AnimGraphNode_RotateRootBone, meshToComponent {yaw=-90}}
    BlendPose_1:Pose Link Structure=Rotate Root Bone(BasePose=@n5, Yaw=Get StartAnimDeltaAngleEast() -> (StartAnimDeltaAngleEast:float)) {meshToComponent {yaw=-90}}
    BlendPose_2:Pose Link Structure=Rotate Root Bone(BasePose=@n2, Yaw=Get StartAnimDeltaAngleSouth() -> (StartAnimDeltaAngleSouth:float)) {meshToComponent {yaw=-90}}
    BlendPose_3:Pose Link Structure=Rotate Root Bone(BasePose=@n4, Yaw=Get StartAnimDeltaAngleWest() -> (StartAnimDeltaAngleWest:float)) {meshToComponent {yaw=-90}}
    BlendTime_0:float=@n6
    BlendTime_1:float=@n6
    BlendTime_2:float=@n6
    BlendTime_3:float=@n6
    ActiveEnumValue:Byte=Get LocomotionCardinalDirection() -> (LocomotionCardinalDirection:EFortCardinalDirection Enum=North)
   ) -> (Pose:Pose Link Structure) {
    is=title AnimGraphNode_BlendListByEnum
    blendPose [{linkId=-1, sourceLinkId=-1}, {linkId=-1, sourceLinkId=-1}, {linkId=-1, sourceLinkId=-1}, {linkId=-1, sourceLinkId=-1}]
    blendTime [0.1,0.1,0.1,0.1]
   }
   Output Animation Pose(Result=@n7)
  ]
 }
 JogStopAdditive {
  path=MovementMode_FinalPoseOverride.AnimGraphNode_StateMachine_0.Locomotion.AnimStateNode_2.JogStopAdditive
  nodes [
   n1 = Sequence Player 'GhostRock_Jog_S_Stop_CMM_NoAdditive'() {
    bLoopAnimation=false
    groupName=StopSync
    method=SyncGroup
    sequence=/Game/Animation/Game/MainPlayer/Locomotion/Medium/Male/Jog/GhostRock/LocomotionNoAdditive/GhostRock_Jog_S_Stop_CMM_NoAdditive
   }
   n2 = Sequence Player 'GhostRock_Jog_W_Stop_CMM_NoAdditive'() {
    bLoopAnimation=false
    groupName=StopSync
    method=SyncGroup
    sequence=/Game/Animation/Game/MainPlayer/Locomotion/Medium/Male/Jog/GhostRock/LocomotionNoAdditive/GhostRock_Jog_W_Stop_CMM_NoAdditive
   }
   n3 = Sequence Player 'GhostRock_Jog_E_Stop_CMM_NoAdditive'() {
    bLoopAnimation=false
    groupName=StopSync
    method=SyncGroup
    sequence=/Game/Animation/Game/MainPlayer/Locomotion/Medium/Male/Jog/GhostRock/LocomotionNoAdditive/GhostRock_Jog_E_Stop_CMM_NoAdditive
   }
   n4 = Sequence Player 'GhostRock_Jog_N_Stop_CMM_NoAdditive'() {
    bLoopAnimation=false
    groupName=StopSync
    method=SyncGroup
    sequence=/Game/Animation/Game/MainPlayer/Locomotion/Medium/Male/Jog/GhostRock/LocomotionNoAdditive/GhostRock_Jog_N_Stop_CMM_NoAdditive
   }
   n5 = "Blend Poses (EFortCardinalDirection)"(
    BlendPose_0=@n4
    BlendPose_1=@n3
    BlendPose_2=@n1
    BlendPose_3=@n2
    BlendTime_0=0.2
    BlendTime_1=0.2
    BlendTime_2=0.2
    BlendTime_3=0.2
    ActiveEnumValue=Get StopCardinalDirection() -> (StopCardinalDirection:EFortCardinalDirection Enum=North)
   ) {
    blendType=Linear
    blendPose [{linkId=-1, sourceLinkId=-1}, {linkId=-1, sourceLinkId=-1}, {linkId=-1, sourceLinkId=-1}, {linkId=-1, sourceLinkId=-1}]
    blendTime [0.1,0.1,0.1,0.1]
   }
   Output Animation Pose(Result=@n5)
  ]
 }
 JogPostPivotAdditive {
  path=MovementMode_FinalPoseOverride.AnimGraphNode_StateMachine_0.Locomotion.AnimStateNode_3.JogPostPivotAdditive
  nodes [
   n1 = Sequence Player 'GhostRock_Pivot_WtoE_NoAdditive'() {
    bLoopAnimation=false
    groupName=Locomotion
    groupRole=TransitionLeader
    method=SyncGroup
    sequence=/Game/Animation/Game/MainPlayer/Combat/Gadgets/Medium/Male/GhostRock/NoAdditives/GhostRock_Pivot_WtoE_NoAdditive
   }
   n2 = Sequence Player 'GhostRock_Pivot_StoN_NoAdditive'() {
    bLoopAnimation=false
    groupName=Locomotion
    groupRole=TransitionLeader
    method=SyncGroup
    sequence=/Game/Animation/Game/MainPlayer/Combat/Gadgets/Medium/Male/GhostRock/NoAdditives/GhostRock_Pivot_StoN_NoAdditive
   }
   n3 = Sequence Player 'GhostRock_Pivot_NtoS_NoAdditive'() {
    bLoopAnimation=false
    groupName=Locomotion
    groupRole=TransitionLeader
    method=SyncGroup
    sequence=/Game/Animation/Game/MainPlayer/Combat/Gadgets/Medium/Male/GhostRock/NoAdditives/GhostRock_Pivot_NtoS_NoAdditive
   }
   n4 = Sequence Player 'GhostRock_Pivot_EtoW_NoAdditive'() {
    bLoopAnimation=false
    groupName=Locomotion
    groupRole=TransitionLeader
    method=SyncGroup
    sequence=/Game/Animation/Game/MainPlayer/Combat/Gadgets/Medium/Male/GhostRock/NoAdditives/GhostRock_Pivot_EtoW_NoAdditive
   }
   n5 = Get PivotAnimBlendTime() -> (PivotAnimBlendTime:double)
   n6 = "Blend Poses (EFortCardinalDirection)"(
    BlendPose_0=@n2
    BlendPose_1=@n1
    BlendPose_2=@n3
    BlendPose_3=@n4
    BlendTime_0=@n5
    BlendTime_1=@n5
    BlendTime_2=@n5
    BlendTime_3=@n5
    ActiveEnumValue=Get PostPivotCardinalDirection() -> (PostPivotCardinalDirection:EFortCardinalDirection Enum=North)
   ) {
    blendType=Linear
    blendPose [{linkId=-1, sourceLinkId=-1}, {linkId=-1, sourceLinkId=-1}, {linkId=-1, sourceLinkId=-1}, {linkId=-1, sourceLinkId=-1}]
    blendTime [0.1,0.1,0.1,0.1]
   }
   Output Animation Pose(Result=Rotate Root Bone(BasePose=@n6, Yaw=Get PostPivotAnimDeltaAngle() -> (PostPivotAnimDeltaAngle:float)) {meshToComponent {yaw=-90}})
  ]
 }
 JogPrePivotAdditive {
  path=MovementMode_FinalPoseOverride.AnimGraphNode_StateMachine_0.Locomotion.AnimStateNode_6.JogPrePivotAdditive
  nodes [
   n1 = Sequence Player 'GhostRock_Pivot_WtoE_NoAdditive'() {
    bLoopAnimation=false
    groupRole=TransitionLeader
    sequence=&GhostRock_Pivot_WtoE_NoAdditive
   }
   n2 = Sequence Player 'GhostRock_Pivot_StoN_NoAdditive'() {
    bLoopAnimation=false
    groupRole=TransitionLeader
    sequence=&GhostRock_Pivot_StoN_NoAdditive
   }
   n3 = Sequence Player 'GhostRock_Pivot_NtoS_NoAdditive'() {
    bLoopAnimation=false
    groupRole=TransitionLeader
    sequence=&GhostRock_Pivot_NtoS_NoAdditive
   }
   n4 = Sequence Player 'GhostRock_Pivot_EtoW_NoAdditive'() {
    bLoopAnimation=false
    groupRole=TransitionLeader
    sequence=&GhostRock_Pivot_EtoW_NoAdditive
   }
   n5 = Get PivotAnimBlendTime()
   n6 = "Blend Poses (EFortCardinalDirection)"(
    BlendPose_0=@n2
    BlendPose_1=@n1
    BlendPose_2=@n3
    BlendPose_3=@n4
    BlendTime_0=@n5
    BlendTime_1=@n5
    BlendTime_2=@n5
    BlendTime_3=@n5
    ActiveEnumValue=Get PrePivotCardinalDirection() -> (PrePivotCardinalDirection:EFortCardinalDirection Enum=North)
   ) {
    blendType=Linear
    blendPose [{linkId=-1, sourceLinkId=-1}, {linkId=-1, sourceLinkId=-1}, {linkId=-1, sourceLinkId=-1}, {linkId=-1, sourceLinkId=-1}]
    blendTime [0.1,0.1,0.1,0.1]
   }
   Output Animation Pose(Result=@n6)
  ]
 }
 MovementAdditive {
  path=MovementMode_FinalPoseOverride.AnimGraphNode_StateMachine_0.Locomotion.AnimStateNode_7.MovementAdditive
  nodes [
   Output Animation Pose(Result=Blendspace Player 'GhostRock_Jog_FullBody_CMM_BS'!(
     X=Get LocalVelocityYawAngleMinusMeleeTwist() -> (LocalVelocityYawAngleMinusMeleeTwist:float)
     Y=Get ClampedSpeed2D() -> (ClampedSpeed2D:double)
     PlayRate:float=Get ClampedPlayRate() -> (ClampedPlayRate:double)
     StartPosition:float=Get JogStartPosition() -> (JogStartPosition:double)
    ) -> (Pose) {blendSpace=/Game/Animation/Game/MainPlayer/Locomotion/Medium/Male/Jog/GhostRock/LocomotionNoAdditive/GhostRock_Jog_FullBody_CMM_BS})
   Get Speed2D()
  ]
 }
 Transition {
  path=MovementMode_FinalPoseOverride.AnimGraphNode_StateMachine_0.Locomotion.AnimStateTransitionNode_0.Transition
  nodes [Result(bCanEnterTransition=Get bShouldPlayStartTransition() -> (bShouldPlayStartTransition:bool))]
 }
 AnimationTransitionGraph_114 {
  path=MovementMode_FinalPoseOverride.AnimGraphNode_StateMachine_0.Locomotion.AnimStateTransitionNode_10.AnimationTransitionGraph_114
  nodes [Result(bCanEnterTransition=Get bTransition_LowerBody_Movement() -> (bTransition_LowerBody_Movement:bool))]
 }
 Transition {
  path=MovementMode_FinalPoseOverride.AnimGraphNode_StateMachine_0.Locomotion.AnimStateTransitionNode_11.Transition
  nodes [Result(bCanEnterTransition=Get bShouldEarlyOutStopState() -> (bShouldEarlyOutStopState:bool))]
 }
 Transition {
  path=MovementMode_FinalPoseOverride.AnimGraphNode_StateMachine_0.Locomotion.AnimStateTransitionNode_12.Transition
  nodes [Result(bCanEnterTransition=Get bShouldPlayStopTransition() -> (bShouldPlayStopTransition:bool))]
 }
 Transition {path=MovementMode_FinalPoseOverride.AnimGraphNode_StateMachine_0.Locomotion.AnimStateTransitionNode_13.Transition, nodes [Result()]}
 Transition {path=MovementMode_FinalPoseOverride.AnimGraphNode_StateMachine_0.Locomotion.AnimStateTransitionNode_2.Transition, nodes [Result()]}
 AnimationTransitionGraph_0 {
  path=MovementMode_FinalPoseOverride.AnimGraphNode_StateMachine_0.Locomotion.AnimStateTransitionNode_25.AnimationTransitionGraph_0
  nodes [Result(bCanEnterTransition=Get bShouldPlayPostPivotTransition() -> (bShouldPlayPostPivotTransition:bool))]
 }
 Transition {
  path=MovementMode_FinalPoseOverride.AnimGraphNode_StateMachine_0.Locomotion.AnimStateTransitionNode_26.Transition
  nodes [Result(bCanEnterTransition=Get bShouldEarlyOutPivotState() -> (bShouldEarlyOutPivotState:bool))]
 }
 Transition {path=MovementMode_FinalPoseOverride.AnimGraphNode_StateMachine_0.Locomotion.AnimStateTransitionNode_27.Transition, nodes [Result()]}
 AnimationTransitionGraph_0 {
  path=MovementMode_FinalPoseOverride.AnimGraphNode_StateMachine_0.Locomotion.AnimStateTransitionNode_28.AnimationTransitionGraph_0
  nodes [Result(bCanEnterTransition=Get bShouldPlayPivotTransition() -> (bShouldPlayPivotTransition:bool))]
 }
 Transition {
  path=MovementMode_FinalPoseOverride.AnimGraphNode_StateMachine_0.Locomotion.AnimStateTransitionNode_3.Transition
  nodes [Result(bCanEnterTransition=Get bShouldEarlyOutStartState() -> (bShouldEarlyOutStartState:bool))]
 }
 Transition {
  path=MovementMode_FinalPoseOverride.AnimGraphNode_StateMachine_0.Locomotion.AnimStateTransitionNode_4.Transition
  nodes [Result(bCanEnterTransition=Get bShouldPlayPivotTransition())]
 }
 Transition {
  path=MovementMode_FinalPoseOverride.AnimGraphNode_StateMachine_0.Locomotion.AnimStateTransitionNode_5.Transition
  nodes [Result(bCanEnterTransition=Get bShouldEarlyOutPivotState())]
 }
 Transition {
  path=MovementMode_FinalPoseOverride.AnimGraphNode_StateMachine_0.Locomotion.AnimStateTransitionNode_6.Transition
  nodes [Result(bCanEnterTransition=Get bShouldPlayPivotTransition())]
 }
 Transition {
  path=MovementMode_FinalPoseOverride.AnimGraphNode_StateMachine_0.Locomotion.AnimStateTransitionNode_7.Transition
  nodes [Result(bCanEnterTransition=Get bShouldPlayStopTransition())]
 }
 Transition {path=MovementMode_FinalPoseOverride.AnimGraphNode_StateMachine_0.Locomotion.AnimStateTransitionNode_8.Transition, nodes [Result()]}
 AnimationTransitionGraph_113 {
  path=MovementMode_FinalPoseOverride.AnimGraphNode_StateMachine_0.Locomotion.AnimStateTransitionNode_9.AnimationTransitionGraph_113
  nodes [Result(bCanEnterTransition=NOT Boolean(A=Get bTransition_LowerBody_Movement()))]
 }
 LocomotionLowerBody {
  entry=Turns
  path=MovementMode_FinalPoseOverride.AnimGraphNode_StateMachine_1.LocomotionLowerBody
  ids=id
  nodes [
   AnimStateNode_0 = Turns() {is=title AnimStateNode}
   AnimStateNode_1 = Idles() {is=title AnimStateNode}
   AnimStateNode_2 = JogStop() {is=title AnimStateNode}
   AnimStateTransitionNode_0 = AnimStateTransitionNode() {bAutomaticRuleBasedOnSequencePlayerInState=true}
   AnimStateTransitionNode_1 = AnimStateTransitionNode() {bAutomaticRuleBasedOnSequencePlayerInState=true}
  ]
  links [
   {
    dst=JogStop
    rule=MovementMode_FinalPoseOverride.AnimGraphNode_StateMachine_1.LocomotionLowerBody.AnimStateTransitionNode_1.Transition
    src=Idles
    via=AnimStateTransitionNode_1
   }
   {
    dst=Idles
    rule=MovementMode_FinalPoseOverride.AnimGraphNode_StateMachine_1.LocomotionLowerBody.AnimStateTransitionNode_0.Transition
    src=Turns
    via=AnimStateTransitionNode_0
   }
  ]
 }
 Turns {
  path=MovementMode_FinalPoseOverride.AnimGraphNode_StateMachine_1.LocomotionLowerBody.AnimStateNode_0.Turns
  nodes [Output Animation Pose(Result=Local Space Ref Pose())]
 }
 Idles {
  path=MovementMode_FinalPoseOverride.AnimGraphNode_StateMachine_1.LocomotionLowerBody.AnimStateNode_1.Idles
  nodes [Output Animation Pose(Result=Local Space Ref Pose())]
 }
 JogStop {
  path=MovementMode_FinalPoseOverride.AnimGraphNode_StateMachine_1.LocomotionLowerBody.AnimStateNode_2.JogStop
  nodes [Output Animation Pose(Result=Local Space Ref Pose())]
 }
 Transition {path=MovementMode_FinalPoseOverride.AnimGraphNode_StateMachine_1.LocomotionLowerBody.AnimStateTransitionNode_0.Transition, nodes [Result()]}
 Transition {path=MovementMode_FinalPoseOverride.AnimGraphNode_StateMachine_1.LocomotionLowerBody.AnimStateTransitionNode_1.Transition, nodes [Result()]}
 FullBodyAdditives {
  entry=IdleAdditive
  path=MovementMode_FinalPoseOverride.AnimGraphNode_StateMachine_2.FullBodyAdditives
  ids=id
  nodes [AnimStateNode_0 = IdleAdditive()]
 }
 IdleAdditive {
  path=MovementMode_FinalPoseOverride.AnimGraphNode_StateMachine_2.FullBodyAdditives.AnimStateNode_0.IdleAdditive
  nodes [Output Animation Pose(Result=Additive Identity Pose())]
 }
 MovementMode_FullBodyOverride {nodes [
   Input Pose() {name=InputPoseFullBody}
   Output Pose() {layerGroup=MovementMode}
  ]}
 MovementMode_InAirOverride {nodes [
   Input Pose() {name=InSource_Pose}
   Input Pose() {name=InAirPose}
   Output Pose() {layerGroup=MovementMode}
  ]}
 MovementMode_LocomotionAdditives {nodes [
   Input Pose() {name=InLocomotionAdditivesPose}
   Output Pose() {layerGroup=MovementMode}
  ]}
 MovementMode_LowerBodyOverride {nodes [
   Input Pose() {name=InLowerBodyPose}
   Input Pose() {name=InSourcePose}
   Output Pose() {layerGroup=MovementMode}
  ]}
 MovementMode_OffScreen {nodes [
   Input Pose() {name=InPoseOffscreen}
   Output Pose() {layerGroup=MovementMode}
  ]}
 MovementMode_SourcePose {nodes [
   Input Pose() {name=InSourcePoseBase}
   Output Pose() {layerGroup=MovementMode}
  ]}
]
variables [
 bIsGhost {type=boolean}
 ButterFlyLocation {type=Vector}
 ButterflyLookAtPitch {type=number}
 ButterflyLookAtYaw {type=number}
 ClampedPlayRate {type=number}
 ClampedSpeed2D {type=number}
 ClampedStartPlayRate {type=number}
 EnableLeanAdditives {type=boolean}
 JogStartPosition {type=number, default=0.625}
 MenuScreenDispatcher {default=()}
 Pawn {type=FortPlayerPawn}
 PivotAnimBlendTime {type=number, default=0.1}
 ShouldLookAtButterfly {type=boolean}
 StartAnimBlendTime {type=number, default=0.33}
 UpperBodyNoAdditivesMaskWeight {type=number}
]
