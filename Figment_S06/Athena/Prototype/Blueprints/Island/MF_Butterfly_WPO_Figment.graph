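# Material function that computes a flapping butterfly-wing world position offset, with matching corrected normals.
# An externally supplied FlapTime plus TimeOffset drives a sine between FlapMin and FlapMax, with vertex green and UV1 phase-shifting it to bend the wing.
# A sphere mask on UV1 adds a secondary phase-shifted center flap, and vertex red mirrors the angle per side.
# That angle drives RotateAboutAxis around Pivot and Axis, and FixRotateAboutAxisNormals for tangent and world normals.
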MaterialFunction /Figment_S06/Athena/Prototype/Blueprints/Island/MF_Butterfly_WPO_Figment
graph {nodes [
  ActorPositionWS()
  n1 = VectorParameter() {parameterName=Axis, defaultValue=<rgba 0 0 1 1>}
  n2 = VertexColor()
  n3 = ScalarParameter() {defaultValue=-0.24, parameterName=FlapMin}
  n4 = ScalarParameter() {defaultValue=0.15, parameterName=FlapMax}
  n5 = ScalarParameter() {parameterName=FlapTime}
  n6 = FunctionInput(Preview=Constant()) {bUsePreviewValueAsDefault=true, inputName=TimeOffset, inputType=FunctionInput_Scalar}
  n7 = Multiply(A=@n2.G) {constB=0.5}
  n8 = TextureCoordinate() {coordinateIndex=1}
  n9 = Multiply(
   A=Power(Base=ComponentMask(None=@n8) {g=true}, Exp=ScalarParameter() {defaultValue=1, parameterName=Top/Bottom Wing Mask Power}) {constExponent=0.7}
   B=ScalarParameter() {defaultValue=-0.1, parameterName=Top/Bottom Wing Offset}
  )
  n10 = Multiply(
   A=ScalarParameter() {parameterName=WingBend}
   B=FunctionInput(Preview=Constant() {r=1}) {bUsePreviewValueAsDefault=true, inputName=WingBendMult, inputType=FunctionInput_Scalar}
  )
  n11 = Add(A=Multiply(A=ScalarParameter() {defaultValue=2, parameterName=FlapSpeed}, B=Add(A=@n5, B=@n6)), B=Multiply(A=Add(A=@n7, B=@n9), B=@n10))
  n12 = ConstantBiasScale(None=Sine(None=@n11))
  n13 = Add(A=@n11) {constB=0.25}
  n14 = LinearInterpolate(Alpha=ConstantBiasScale(None=Sine(None=@n13))) {constA=-1}
  n15 = TextureCoordinate() {coordinateIndex=1}
  n16 = Constant2Vector() {g=0.5, r=1}
  n17 = Constant() {r=0.7}
  n18 = Multiply(
   A=Multiply(
    A=SphereMask(A=@n15, B=@n16, Radius=@n17) {attenuationRadius=0.4, hardnessPercent=0}
    B=ScalarParameter() {defaultValue=0.12, parameterName=CenterFlapAdd}
   ) {constB=0.1}
   B=@n10
  )
  n19 = Multiply(A=LinearInterpolate(Alpha=@n2.R) {constA=-1}, B=Add(A=LinearInterpolate(A=@n3, B=@n4, Alpha=@n12), B=Multiply(A=@n14, B=@n18)))
  FunctionOutput(None=RotateAboutAxis(NormalizedRotationAxis=@n1.RGB, RotationAngle=@n19, PivotPoint=VectorParameter() {parameterName=Pivot}.RGB, Position=WorldPosition().XYZ)) {outputName=WPO}
  n20 = MaterialFunctionCall(Rotation Axis=@n1.RGB, Rotation Angle=@n19) {materialFunction=&FixRotateAboutAxisNormals}
  FunctionOutput(None=@n20.New Tangent Space Vertex Normal) {outputName=TangentNormal}
  FunctionOutput(None=@n20.New World Space Vertex Normal) {outputName=WorldNormal}
  Multiply(A=Add(A=Time(), B=@n6)) {constB=0.3}
  ObjectPositionWS()
 ]}
