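# Translucent, two-sided unlit material for glowing cracked-glass butterfly shards in the Figment island prototype.
# Layered crack textures, a UV gradient between base and crack colors, and a tinted, Fresnel-weighted cubemap reflection build the glow, with blended crack normals.
# MF_Butterfly_WPO_Figment supplies vertex motion, plus a local-space offset built from UV3 data.
# The Opacity parameter drives transparency, slides the tiled cracks and shrinks that offset, pulling scattered pieces into place.
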Material /Figment_S06/Athena/Prototype/Blueprints/Island/M_Butterfly_Shard_Figment
graph {
 properties {
  blendMode=BLEND_Translucent
  ditherOpacityMask=true
  pixelDepthOffsetMode=PDOM_Legacy
  refractionMethod=RM_IndexOfRefraction
  shadingModel=MSM_Unlit
  twoSided=true
  usedShadingModels=Unlit
 }
 nodes [
  n1 = TextureSampleParameterCube!(UVs=ReflectionVectorWS() {bNormalizeCustomWorldNormal=true}, Apply View MipBias) {parameterName=Cubemap, samplerType=SAMPLERTYPE_LinearColor, texture=/Game/Packages/Fortress_Sky/TexturesHDR/T_AthenaSkylight}
  n2 = VectorParameter() {parameterName=CubemapTint, defaultValue=<rgba 0.4197 0.7965 1 1>}
  n3 = Fresnel() {baseReflectFraction=0, exponent=3}
  n4 = Constant2Vector() {g=0.25}
  n5 = ScalarParameter() {defaultValue=1, parameterName=Opacity}
  n6 = Add(A=TextureCoordinate() {coordinateIndex=2, uTiling=1.25, vTiling=0.5}, B=Multiply(A=TextureCoordinate() {coordinateIndex=1}) {constB=0.5})
  n7 = TextureSample(UVs=Add(A=LinearInterpolate(A=@n4, Alpha=@n5) {constA=0.5, constB=0}, B=@n6)) {texture=/Game/Effects/Fort_Effects/Textures/Tile/T_Glass_Cracks_M}
  n8 = TextureCoordinate()
  n9 = VectorParameter() {parameterName=CrackColor, defaultValue=<rgba 0.2058 0.778 1 0>}
  n10 = Multiply(
   A=Add(A=Multiply(A=@n7.G) {constB=0.6}, B=TextureSample(UVs=@n8) {texture=&T_Glass_Cracks_M}.G)
   B=Multiply(A=@n9.RGB) {constB=8}
  )
  n11 = ComponentMask(None=TextureCoordinate() {coordinateIndex=2}) {g=true}
  n12 = Power(Base=@n11)
  n13 = LinearInterpolate(A=VectorParameter() {parameterName=BaseColor, defaultValue=<rgba 0.1405 0.1916 0.5 0>}.RGB, B=@n9.RGB, Alpha=@n12)
  Add(
   A=Multiply(A=Multiply(A=@n1.RGB, B=@n2.RGB) {constB=0.7}, B=Fresnel() {baseReflectFraction=0.3, exponent=3})
   B=Multiply(A=LinearInterpolate(Alpha=@n3) {constA=1, constB=2}, B=Add(A=@n10, B=@n13))
  )
  n14 = MaterialFunctionCall!(TimeOffset, WingBendMult) {materialFunction=&/MF_Butterfly_WPO_Figment}
  n15 = ScalarParameter() {defaultValue=800, parameterName=Offset}
  n16 = Constant()
  n17 = TextureCoordinate() {coordinateIndex=3}
  n18 = ComponentMask(None=@n17) {r=true}
  n19 = ComponentMask(None=@n17) {g=true}
  n20 = Add(A=AppendVector(A=AppendVector(A=@n16, B=@n18), B=OneMinus(None=@n19)), B=Constant3Vector() {constant=<rgba 0 -0.5 -0.5 0>}) {constB=-0.5}
  Add(
   A=@n14.WPO
   B=Multiply(A=LinearInterpolate(A=@n15, Alpha=@n5) {constA=1, constB=0}, B=Transform(None=@n20) {transformSourceType=TRANSFORMSOURCE_Local})
  ) {constB=0}
  Constant() {r=1}
  MaterialFunctionCall() {materialFunction=&DebugTimeSine}
  n21 = TextureSampleParameter2D() {parameterName=Normals, samplerType=SAMPLERTYPE_Normal, texture=/Game/Effects/Fort_Effects/Textures/Tile/T_Glass_Cracks_N}
  n22 = Constant3Vector() {constant=<rgba 2 2 1 0>}
  n23 = TextureSampleParameter2D(UVs=@n6) {parameterName=NormalsTile, samplerType=SAMPLERTYPE_Normal, texture=&T_Glass_Cracks_N}
  n24 = Constant3Vector() {constant=<rgba 1 1 0 0>}
  MaterialFunctionCall(BaseNormal=@n14.WorldNormal, AdditionalNormal=Add(A=Multiply(A=@n21.RGB, B=@n22), B=Multiply(A=@n23.RGB, B=@n24))) {materialFunction=&BlendAngleCorrectedNormals}
  n25 = Power(Base=@n11) {constExponent=10}
  Multiply(A=LinearInterpolate(Alpha=@n12) {constA=0.2, constB=1.05}, B=LinearInterpolate(A=@n5, Alpha=Multiply(A=@n5, B=@n25))) {constB=1.5}
  Power(Base=@n5)
 ]
}
